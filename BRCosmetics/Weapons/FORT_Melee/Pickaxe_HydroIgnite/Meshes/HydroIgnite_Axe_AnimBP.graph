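# Animation Blueprint for the HydroIgnite pickaxe that continuously spins the axe's hammer head and four drill bones.
# Each update, the EventGraph derives two roll rotations from game time, with the drills turning faster than the hammer.
# The AnimGraph layers these onto the input pose through chained additive, bone-space Modify Bone nodes, with all four drills sharing one rotation.
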AnimBlueprint /BRCosmetics/Weapons/FORT_Melee/Pickaxe_HydroIgnite/Meshes/HydroIgnite_Axe_AnimBP
target_skeleton=&/HydroIgnite_Axe_Skeleton
graphs [
 AnimGraph {nodes [
   n1 = "Transform (Modify) Bone - Bone: HammerHead_Base_jnt"(
    ComponentPose=Local To Component(LocalPose=Input Pose())
    Alpha=1
    Translation="0,0,0"
    Rotation=Get HammerSpin_Rotation() -> (HammerSpin_Rotation:Rotator)
    Scale="1,1,1"
   ) {
    rotationMode=BMM_Additive
    rotationSpace=BCS_BoneSpace
    translationMode=BMM_Additive
    translationSpace=BCS_BoneSpace
    alphaBoolBlend {blendOption=Linear}
    boneToModify {boneName=HammerHead_Base_jnt}
   }
   n2 = Get DrillSpin_Rotation() -> (DrillSpin_Rotation:Rotator)
   n3 = "Transform (Modify) Bone - Bone: Drill_04_Base_jnt"(ComponentPose=@n1, Alpha=1, Translation="0,0,0", Rotation=@n2, Scale="1,1,1") {
    rotationMode=BMM_Additive
    rotationSpace=BCS_BoneSpace
    translationMode=BMM_Additive
    translationSpace=BCS_BoneSpace
    alphaBoolBlend {blendOption=Linear}
    boneToModify {boneName=Drill_04_Base_jnt}
   }
   n4 = "Transform (Modify) Bone - Bone: Drill_03_Base_jnt"(ComponentPose=@n3, Alpha=1, Translation="0,0,0", Rotation=@n2, Scale="1,1,1") {
    rotationMode=BMM_Additive
    rotationSpace=BCS_BoneSpace
    translationMode=BMM_Additive
    translationSpace=BCS_BoneSpace
    alphaBoolBlend {blendOption=Linear}
    boneToModify {boneName=Drill_03_Base_jnt}
   }
   n5 = "Transform (Modify) Bone - Bone: Drill_02_Base_jnt"(ComponentPose=@n4, Alpha=1, Translation="0,0,0", Rotation=@n2, Scale="1,1,1") {
    rotationMode=BMM_Additive
    rotationSpace=BCS_BoneSpace
    translationMode=BMM_Additive
    translationSpace=BCS_BoneSpace
    alphaBoolBlend {blendOption=Linear}
    boneToModify {boneName=Drill_02_Base_jnt}
   }
   n6 = "Transform (Modify) Bone - Bone: Drill_01_Base_jnt"(ComponentPose=@n5, Alpha=1, Translation="0,0,0", Rotation=@n2, Scale="1,1,1") {
    rotationMode=BMM_Additive
    rotationSpace=BCS_BoneSpace
    translationMode=BMM_Additive
    translationSpace=BCS_BoneSpace
    alphaBoolBlend {blendOption=Linear}
    boneToModify {boneName=Drill_01_Base_jnt}
   }
   Output Pose(Result=Component To Local(ComponentPose=@n6)) {layerGroup=None}
  ]}
 EventGraph {nodes [
   Event BlueprintUpdateAnimation()
   > n1 = Sequence()
   n2 = GetGameTimeInSeconds()
   > Set HammerSpin_Rotation(HammerSpin_Rotation:Rotator=MakeRotator(Roll=float * float(A=@n2, B=60) {is=title K2Node_PromotableOperator})) -> (Output_Get:Rotator)
   Set DrillSpin_Rotation(execute=@n1.then_1, DrillSpin_Rotation:Rotator=MakeRotator(Roll=float * float(A=@n2, B=150))) -> (Output_Get:Rotator)
   TryGetPawnOwner()
  ]}
]
variables [
 DrillSpin_Rotation {type=Rotator}
 HammerSpin_Rotation {type=Rotator}
]
